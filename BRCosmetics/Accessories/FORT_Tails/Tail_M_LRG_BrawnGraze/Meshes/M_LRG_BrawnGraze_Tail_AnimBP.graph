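# Animation Blueprint for the BrawnGraze tail cosmetic on the Fortnite base tail skeleton, which as wired passes its input pose straight to the output.
# An additive bone-space Transform Bone chain covers seven tail joints, but has zero offsets and does not reach Output Pose.
# The EventGraph computes time-based sine pitch and yaw sway into four Rotator variables, but the update event is unconnected and no AnimGraph node reads them.
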AnimBlueprint /BRCosmetics/Accessories/FORT_Tails/Tail_M_LRG_BrawnGraze/Meshes/M_LRG_BrawnGraze_Tail_AnimBP
target_skeleton=/Game/Accessories/FORT_Tails/Common/Fortnite_Base_Tail_Skeleton
graphs [
 AnimGraph {nodes [
   n1 = "Transform (Modify) Bone - Bone: C_Tail_A_Base_Jnt"(ComponentPose=Local To Component(), Alpha=1, Translation="0,0,0", Rotation="0,0,0", Scale="1,1,1") {
    rotationMode=BMM_Additive
    rotationSpace=BCS_BoneSpace
    translationMode=BMM_Additive
    translationSpace=BCS_BoneSpace
    alphaBoolBlend {blendInTime=1, blendOutTime=1}
    boneToModify {boneName=C_Tail_A_Base_Jnt}
   }
   n2 = "Transform (Modify) Bone - Bone: C_Tail_A_1_Jnt"(ComponentPose=@n1, Alpha=1, Translation="0,0,0", Rotation="0,0,0", Scale="1,1,1") {
    rotationMode=BMM_Additive
    rotationSpace=BCS_BoneSpace
    translationMode=BMM_Additive
    translationSpace=BCS_BoneSpace
    alphaBoolBlend {blendInTime=1, blendOutTime=1}
    boneToModify {boneName=C_Tail_A_1_Jnt}
   }
   n3 = "Transform (Modify) Bone - Bone: C_Tail_A_2_Jnt"(ComponentPose=@n2, Alpha=1, Translation="0,0,0", Rotation="0,0,0", Scale="1,1,1") {
    rotationMode=BMM_Additive
    rotationSpace=BCS_BoneSpace
    translationMode=BMM_Additive
    translationSpace=BCS_BoneSpace
    alphaBoolBlend {blendInTime=1, blendOutTime=1}
    boneToModify {boneName=C_Tail_A_2_Jnt}
   }
   n4 = "Transform (Modify) Bone - Bone: C_Tail_A_3_Jnt"(ComponentPose=@n3, Alpha=1, Translation="0,0,0", Rotation="0,0,0", Scale="1,1,1") {
    rotationMode=BMM_Additive
    rotationSpace=BCS_BoneSpace
    translationMode=BMM_Additive
    translationSpace=BCS_BoneSpace
    alphaBoolBlend {blendInTime=1, blendOutTime=1}
    boneToModify {boneName=C_Tail_A_3_Jnt}
   }
   n5 = "Transform (Modify) Bone - Bone: C_Tail_A_4_Jnt"(ComponentPose=@n4, Alpha=1, Translation="0,0,0", Rotation="0,0,0", Scale="1,1,1") {
    rotationMode=BMM_Additive
    rotationSpace=BCS_BoneSpace
    translationMode=BMM_Additive
    translationSpace=BCS_BoneSpace
    alphaBoolBlend {blendInTime=1, blendOutTime=1}
    boneToModify {boneName=C_Tail_A_4_Jnt}
   }
   n6 = "Transform (Modify) Bone - Bone: C_Tail_A_5_Jnt"(ComponentPose=@n5, Alpha=1, Translation="0,0,0", Rotation="0,0,0", Scale="1,1,1") {
    rotationMode=BMM_Additive
    rotationSpace=BCS_BoneSpace
    translationMode=BMM_Additive
    translationSpace=BCS_BoneSpace
    alphaBoolBlend {blendInTime=1, blendOutTime=1}
    boneToModify {boneName=C_Tail_A_5_Jnt}
   }
   n7 = "Transform (Modify) Bone - Bone: C_Tail_A_6_Jnt"(ComponentPose=@n6, Alpha=1, Translation="0,0,0", Rotation="0,0,0", Scale="1,1,1") {
    rotationMode=BMM_Additive
    rotationSpace=BCS_BoneSpace
    translationMode=BMM_Additive
    translationSpace=BCS_BoneSpace
    alphaBoolBlend {blendInTime=1, blendOutTime=1}
    boneToModify {boneName=C_Tail_A_6_Jnt}
   }
   Component To Local(ComponentPose=@n7)
   Output Pose(Result=Input Pose())
  ]}
 EventGraph {nodes [
   n1 = GetRealTimeSeconds()
   n2 = "Sin (Radians)"(A=float * float(A=float * float(A=2.2, B=2.2), B=@n1))
   Set Rotator(Rotator:Rotator=MakeRotator(Pitch=float * float(A=@n2, B=2))) -> (Output_Get:Rotator)
   n3 = "Sin (Radians)"(A=float * float(A=float * float(A=1.8, B=1.8), B=@n1))
   > Set Rotator_1(Rotator_1:Rotator=MakeRotator(Pitch=float * float(A=@n3, B=1.8))) -> (Output_Get:Rotator)
   n4 = "Sin (Radians)"(A=float * float(A=float * float(A=1.4, B=1.4), B=@n1))
   > Set Rotator_2(Rotator_2:Rotator=MakeRotator(Pitch=float * float(A=@n4, B=1.4))) -> (Output_Get:Rotator)
   n5 = "Sin (Radians)"(A=float * float(A=float * float(A=2.2, B=2.2), B=@n1))
   > Set Rotator_3(Rotator_3:Rotator=MakeRotator(Yaw=float * float(A=@n5, B=2))) -> (Output_Get:Rotator)
   Event BlueprintUpdateAnimation()
  ]}
]
variables [
 Rotator {type=Rotator}
 Rotator_1 {type=Rotator}
 Rotator_2 {type=Rotator}
 Rotator_3 {type=Rotator}
]
