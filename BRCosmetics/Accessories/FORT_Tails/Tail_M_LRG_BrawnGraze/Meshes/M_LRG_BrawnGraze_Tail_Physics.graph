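# Physics asset for the large male BrawnGraze tail cosmetic accessory, giving its tail joints simple collision and physics bodies.
# Seven capsules follow the bone chain from C_Tail_A_Base_Jnt to C_Tail_A_6_Jnt, growing generally thicker toward the tip.
# Six constraints link each joint to the one before it, forming a single serial chain.
# Swing and twist are Free on every constraint, so the stored 45-degree angular limits are not enforced.
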PhysicsAsset /BRCosmetics/Accessories/FORT_Tails/Tail_M_LRG_BrawnGraze/Meshes/M_LRG_BrawnGraze_Tail_Physics
bodies {
 C_Tail_A_1_Jnt [{
   shapeType=Capsule
   length=5.3229
   radius=6.7846
   shapeName=C_Tail_A_1_Jnt_capsule
   center=<4.7221 0.0302 -0.15>
   rotation=<rot 0.0076 91.5219 -90>
  }]
 C_Tail_A_2_Jnt [{
   shapeType=Capsule
   length=9.403
   radius=6.919
   shapeName=C_Tail_A_2_Jnt_capsule
   center=<5.0719 0.0345 -0.1295>
   rotation=<rot 0.1535 -89.5217 -90.008>
  }]
 C_Tail_A_3_Jnt [{
   shapeType=Capsule
   length=7.3573
   radius=7.5048
   shapeName=C_Tail_A_3_Jnt_capsule
   center=<4.1038 -0.2354 -0.0933>
   rotation=<rot 0.1053 86.9411 -90.0188>
  }]
 C_Tail_A_4_Jnt [{
   shapeType=Capsule
   length=7.5857
   radius=7.296
   shapeName=C_Tail_A_4_Jnt_capsule
   center=<6.3221 0.0335 -0.0482>
   rotation=<rot 0.0057 90.7274 -90.0002>
  }]
 C_Tail_A_5_Jnt [{
   shapeType=Capsule
   length=9.4413
   radius=8.255
   shapeName=C_Tail_A_5_Jnt_capsule
   center=<5.2204 0.1087 -0.0325>
   rotation=<rot 0.0205 -89.1891 -89.999>
  }]
 C_Tail_A_6_Jnt [{
   shapeType=Capsule
   length=8.5637
   radius=8.559
   shapeName=C_Tail_A_6_Jnt_capsule
   center=<6.0301 0.669 -0.2635>
   rotation=<rot 0.1114 -90.2788 -89.9982>
  }]
 C_Tail_A_Base_Jnt [{
   shapeType=Capsule
   length=5.4681
   radius=5.9295
   shapeName=C_Tail_A_Base_Jnt_capsule
   center=<1.635 -0.2687 -0.0451>
   rotation=<rot 0 -90.0006 -91.1816>
  }]
}
constraints [
 {
  bone1Name=C_Tail_A_1_Jnt
  bone2Name=C_Tail_A_Base_Jnt
  swing1LimitDegrees=45
  swing1Motion=Free
  swing2LimitDegrees=45
  swing2Motion=Free
  twistLimitDegrees=45
  twistMotion=Free
 }
 {
  bone1Name=C_Tail_A_2_Jnt
  bone2Name=C_Tail_A_1_Jnt
  swing1LimitDegrees=45
  swing1Motion=Free
  swing2LimitDegrees=45
  swing2Motion=Free
  twistLimitDegrees=45
  twistMotion=Free
 }
 {
  bone1Name=C_Tail_A_3_Jnt
  bone2Name=C_Tail_A_2_Jnt
  swing1LimitDegrees=45
  swing1Motion=Free
  swing2LimitDegrees=45
  swing2Motion=Free
  twistLimitDegrees=45
  twistMotion=Free
 }
 {
  bone1Name=C_Tail_A_4_Jnt
  bone2Name=C_Tail_A_3_Jnt
  swing1LimitDegrees=45
  swing1Motion=Free
  swing2LimitDegrees=45
  swing2Motion=Free
  twistLimitDegrees=45
  twistMotion=Free
 }
 {
  bone1Name=C_Tail_A_5_Jnt
  bone2Name=C_Tail_A_4_Jnt
  swing1LimitDegrees=45
  swing1Motion=Free
  swing2LimitDegrees=45
  swing2Motion=Free
  twistLimitDegrees=45
  twistMotion=Free
 }
 {
  bone1Name=C_Tail_A_6_Jnt
  bone2Name=C_Tail_A_5_Jnt
  swing1LimitDegrees=45
  swing1Motion=Free
  swing2LimitDegrees=45
  swing2Motion=Free
  twistLimitDegrees=45
  twistMotion=Free
 }
]
