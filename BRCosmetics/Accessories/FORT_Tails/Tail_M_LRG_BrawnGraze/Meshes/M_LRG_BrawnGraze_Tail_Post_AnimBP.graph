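# Post-process AnimBP for the BrawnGraze large tail cosmetic, adjusting tail pose and adding secondary motion.
# Additive root and tail bone offsets feed seven chained, spring-driven Anim Dynamics joints with tight angular limits, followed by further blendable offsets.
# On customization load, it flags cape or long backblings as capes.
# Each update combines that flag with sliding, tactical sprint, skydiving, parachute and front-end idle-curve states into pose-condition variables.
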AnimBlueprint /BRCosmetics/Accessories/FORT_Tails/Tail_M_LRG_BrawnGraze/Meshes/M_LRG_BrawnGraze_Tail_Post_AnimBP
target_skeleton=/Game/Accessories/FORT_Tails/Common/Fortnite_Base_Tail_Skeleton
graphs [
 AnimGraph {nodes [
   n1 = "Transform (Modify) Bone - Bone: C_Root_Main_Root_Jnt"(ComponentPose=Local To Component(LocalPose=Input Pose()), Alpha=1, Translation="0,-1.5,-1.5", Rotation="0,0,30", Scale="1,1,1") {
    rotationMode=BMM_Additive
    rotationSpace=BCS_BoneSpace
    translationMode=BMM_Additive
    translationSpace=BCS_BoneSpace
    alphaScaleBiasClamp {bias=1, scale=-1}
    boneToModify {boneName=C_Root_Main_Root_Jnt}
    rotation {roll=30}
    translation {y=-1.5, z=-1.5}
   }
   n2 = "Transform (Modify) Bone - Bone: C_Tail_A_Base_Jnt"(ComponentPose=@n1, Alpha=1, Translation="0,0,0", Rotation="0,5,0", Scale="1,1,1") {
    rotationMode=BMM_Additive
    rotationSpace=BCS_BoneSpace
    alphaScaleBiasClamp {bias=1, scale=-1}
    boneToModify {boneName=C_Tail_A_Base_Jnt}
    rotation {yaw=5}
   }
   n3 = "Transform (Modify) Bone - Bone: C_Root_Main_Root_Jnt"(ComponentPose=@n2, Alpha=1, Translation="0,0,0", Rotation="0,0,70", Scale="1,1,1") {
    rotationMode=BMM_Additive
    translationMode=BMM_Additive
    alphaBoolBlend {blendInTime=0.5, blendOutTime=0.8}
    alphaScaleBiasClamp {bInterpResult=true, interpSpeedDecreasing=15, interpSpeedIncreasing=15}
    boneToModify {boneName=C_Root_Main_Root_Jnt}
    rotation {roll=70}
   }
   n4 = "Transform (Modify) Bone - Bone: C_Tail_A_1_Jnt"(ComponentPose=@n3, Alpha=1, Translation="0,0,0", Rotation="20,0,0", Scale="1,1,1") {
    rotationMode=BMM_Additive
    rotationSpace=BCS_BoneSpace
    translationMode=BMM_Additive
    translationSpace=BCS_BoneSpace
    alphaBoolBlend {blendInTime=1, blendOutTime=1}
    alphaScaleBiasClamp {bInterpResult=true, interpSpeedDecreasing=15, interpSpeedIncreasing=15}
    boneToModify {boneName=C_Tail_A_1_Jnt}
    rotation {pitch=20}
   }
   n5 = Anim Dynamics - Bone: C_Tail_A_Base_Jnt(ComponentPose=@n4, Alpha=1, ExternalForce="0,0,0") {
    angularDampingOverride=1
    angularSpringConstant=20000
    bAngularSpring=true
    bOverrideAngularDamping=true
    bUseGravityOverride=true
    gravityScale=0.35
    simulationSpace=RootRelative
    boundBone {boneName=C_Tail_A_Base_Jnt}
    componentAppliedLinearAccClamp=<2000 2000 2000>
    componentLinearAccScale=<0.5 0.5 0.5>
    componentLinearVelScale=<0.5 0.5 0.5>
    gravityOverride {z=-980}
    physicsBodyDefinitions [{
      boundBone {boneName=C_Tail_A_Base_Jnt}
      boxExtents=<30 30 30>
      constraintSetup {angularLimitsMax=<0 0 10>, angularLimitsMin=<0 0 -3>, angularTarget=<1 0 0>}
      localJointOffset=<2 0 0>
     }]
    simSpaceSettings {maxAngularAcceleration=25, maxAngularVelocity=5, simSpaceAngularAlpha=1}
   }
   n6 = Anim Dynamics - Bone: C_Tail_A_1_Jnt(ComponentPose=@n5, Alpha=1, ExternalForce="0,0,0") {
    angularDampingOverride=1
    angularSpringConstant=20000
    bAngularSpring=true
    bOverrideAngularDamping=true
    bUseGravityOverride=true
    gravityScale=0.35
    simulationSpace=RootRelative
    boundBone {boneName=C_Tail_A_1_Jnt}
    componentAppliedLinearAccClamp=<2000 2000 2000>
    componentLinearAccScale=<0.5 0.5 0.5>
    componentLinearVelScale=<0.5 0.5 0.5>
    gravityOverride {z=-980}
    physicsBodyDefinitions [{
      boundBone {boneName=C_Tail_A_1_Jnt}
      boxExtents=<30 30 30>
      constraintSetup {angularLimitsMax=<0 3 5>, angularLimitsMin=<0 -10 -5>, angularTarget=<1 0 0>}
      localJointOffset=<2 0 0>
     }]
    simSpaceSettings {maxAngularAcceleration=25, maxAngularVelocity=5, simSpaceAngularAlpha=1}
   }
   n7 = Anim Dynamics - Bone: C_Tail_A_2_Jnt(ComponentPose=@n6, Alpha=1, ExternalForce="0,0,0") {
    angularDampingOverride=1
    angularSpringConstant=20000
    bAngularSpring=true
    bOverrideAngularDamping=true
    bUseGravityOverride=true
    gravityScale=0.35
    simulationSpace=RootRelative
    boundBone {boneName=C_Tail_A_2_Jnt}
    componentAppliedLinearAccClamp=<2000 2000 2000>
    componentLinearAccScale=<0.5 0.5 0.5>
    componentLinearVelScale=<0.5 0.5 0.5>
    gravityOverride {z=-980}
    physicsBodyDefinitions [{
      boundBone {boneName=C_Tail_A_2_Jnt}
      boxExtents=<30 30 30>
      constraintSetup {angularLimitsMax=<0 3 0>, angularLimitsMin=<0 -10 0>, angularTarget=<1 0 0>}
      localJointOffset=<2 0 0>
     }]
    simSpaceSettings {maxAngularAcceleration=25, maxAngularVelocity=5, simSpaceAngularAlpha=1}
   }
   n8 = Anim Dynamics - Bone: C_Tail_A_3_Jnt(ComponentPose=@n7, Alpha=1, ExternalForce="0,0,0") {
    angularDampingOverride=1
    angularSpringConstant=20000
    bAngularSpring=true
    bOverrideAngularDamping=true
    bUseGravityOverride=true
    gravityScale=0.35
    simulationSpace=RootRelative
    boundBone {boneName=C_Tail_A_3_Jnt}
    componentAppliedLinearAccClamp=<2000 2000 2000>
    componentLinearAccScale=<0.5 0.5 0.5>
    componentLinearVelScale=<0.5 0.5 0.5>
    gravityOverride {z=-980}
    physicsBodyDefinitions [{
      boundBone {boneName=C_Tail_A_3_Jnt}
      boxExtents=<30 30 30>
      constraintSetup {angularLimitsMax=<0 3 0>, angularLimitsMin=<0 -10 0>, angularTarget=<1 0 0>}
      localJointOffset=<2 0 0>
     }]
    simSpaceSettings {maxAngularAcceleration=25, maxAngularVelocity=5, simSpaceAngularAlpha=1}
   }
   n9 = Anim Dynamics - Bone: C_Tail_A_4_Jnt(ComponentPose=@n8, Alpha=1, ExternalForce="0,0,0") {
    angularDampingOverride=1
    angularSpringConstant=20000
    bAngularSpring=true
    bOverrideAngularDamping=true
    bUseGravityOverride=true
    gravityScale=0.35
    simulationSpace=RootRelative
    boundBone {boneName=C_Tail_A_4_Jnt}
    componentAppliedLinearAccClamp=<2000 2000 2000>
    componentLinearAccScale=<0.5 0.5 0.5>
    componentLinearVelScale=<0.5 0.5 0.5>
    gravityOverride {z=-980}
    physicsBodyDefinitions [{
      boundBone {boneName=C_Tail_A_4_Jnt}
      boxExtents=<30 30 30>
      constraintSetup {angularLimitsMax=<0 3 0>, angularLimitsMin=<0 -10 0>, angularTarget=<1 0 0>}
      localJointOffset=<2 0 0>
     }]
    simSpaceSettings {maxAngularAcceleration=25, maxAngularVelocity=5, simSpaceAngularAlpha=1}
   }
   n10 = Anim Dynamics - Bone: C_Tail_A_5_Jnt(ComponentPose=@n9, Alpha=1, ExternalForce="0,0,0") {
    angularDampingOverride=1
    angularSpringConstant=20000
    bAngularSpring=true
    bOverrideAngularDamping=true
    bUseGravityOverride=true
    gravityScale=0.35
    simulationSpace=RootRelative
    boundBone {boneName=C_Tail_A_5_Jnt}
    componentAppliedLinearAccClamp=<2000 2000 2000>
    componentLinearAccScale=<0.5 0.5 0.5>
    componentLinearVelScale=<0.5 0.5 0.5>
    gravityOverride {z=-980}
    physicsBodyDefinitions [{
      boundBone {boneName=C_Tail_A_5_Jnt}
      boxExtents=<30 30 30>
      constraintSetup {angularLimitsMax=<0 3 0>, angularLimitsMin=<0 -10 0>, angularTarget=<1 0 0>}
      localJointOffset=<2 0 0>
     }]
    simSpaceSettings {maxAngularAcceleration=25, maxAngularVelocity=5, simSpaceAngularAlpha=1}
   }
   n11 = Anim Dynamics - Bone: C_Tail_A_6_Jnt(ComponentPose=@n10, Alpha=1, ExternalForce="0,0,0") {
    angularDampingOverride=1
    angularSpringConstant=20000
    bAngularSpring=true
    bOverrideAngularDamping=true
    bUseGravityOverride=true
    gravityScale=0.35
    simulationSpace=RootRelative
    boundBone {boneName=C_Tail_A_6_Jnt}
    componentAppliedLinearAccClamp=<2000 2000 2000>
    componentLinearAccScale=<0.5 0.5 0.5>
    componentLinearVelScale=<0.5 0.5 0.5>
    gravityOverride {z=-980}
    physicsBodyDefinitions [{
      boundBone {boneName=C_Tail_A_6_Jnt}
      boxExtents=<30 30 30>
      constraintSetup {angularLimitsMax=<0 3 0>, angularLimitsMin=<0 -10 0>, angularTarget=<1 0 0>}
      localJointOffset=<2 0 0>
     }]
    simSpaceSettings {maxAngularAcceleration=25, maxAngularVelocity=5, simSpaceAngularAlpha=1}
   }
   n12 = "Transform (Modify) Bone - Bone: C_Root_Main_Root_Jnt"!(ComponentPose=@n11, bAlphaBoolEnabled:bool=True, Translation="0,0,0", Rotation="0,0,-50", Scale="0.9,1,1") -> (Pose) {
    alphaInputType=Bool
    rotationMode=BMM_Additive
    rotationSpace=BCS_BoneSpace
    scaleMode=BMM_Additive
    scaleSpace=BCS_BoneSpace
    translationMode=BMM_Additive
    translationSpace=BCS_BoneSpace
    alphaBoolBlend {blendInTime=0.1, blendOutTime=0.1}
    boneToModify {boneName=C_Root_Main_Root_Jnt}
    rotation {roll=-50}
    scale {x=0.9}
   }
   n13 = "Transform (Modify) Bone - Bone: C_Tail_A_Base_Jnt"!(ComponentPose=@n12, bAlphaBoolEnabled:bool=True, Translation="0,-1,0", Rotation="0,-15,0", Scale="0.9,1,1") -> (Pose) {
    alphaInputType=Bool
    rotationMode=BMM_Additive
    rotationSpace=BCS_BoneSpace
    scaleMode=BMM_Additive
    scaleSpace=BCS_BoneSpace
    translationMode=BMM_Additive
    translationSpace=BCS_BoneSpace
    alphaBoolBlend {blendInTime=0.1, blendOutTime=0.1}
    boneToModify {boneName=C_Tail_A_Base_Jnt}
    rotation {yaw=-15}
    scale {x=0.9}
    translation {y=-1}
   }
   n14 = "Transform (Modify) Bone - Bone: C_Tail_A_1_Jnt"!(ComponentPose=@n13, bAlphaBoolEnabled:bool=True, Translation="0,0,0", Rotation="-10,0,0", Scale="0.95,1,1") -> (Pose) {
    alphaInputType=Bool
    rotationMode=BMM_Additive
    rotationSpace=BCS_BoneSpace
    scaleMode=BMM_Additive
    scaleSpace=BCS_BoneSpace
    translationMode=BMM_Additive
    translationSpace=BCS_BoneSpace
    alphaBoolBlend {blendInTime=0.1, blendOutTime=0.1}
    boneToModify {boneName=C_Tail_A_1_Jnt}
    rotation {pitch=-10}
    scale {x=0.95}
   }
   n15 = "Transform (Modify) Bone - Bone: C_Root_Main_Root_Jnt"(ComponentPose=@n14, bAlphaBoolEnabled=True, Translation="0,0,0", Rotation="0,0,40", Scale="1,1,1") {
    alphaInputType=Bool
    rotationMode=BMM_Additive
    rotationSpace=BCS_BoneSpace
    translationMode=BMM_Additive
    translationSpace=BCS_BoneSpace
    alphaBoolBlend {blendInTime=0.2, blendOutTime=0.1}
    boneToModify {boneName=C_Root_Main_Root_Jnt}
    rotation {roll=40}
   }
   n16 = "Transform (Modify) Bone - Bone: C_Root_Main_Root_Jnt"(ComponentPose=@n15, bAlphaBoolEnabled=True, Translation="0,0,0", Rotation="0,0,9", Scale="0.95,1,1") {
    alphaInputType=Bool
    rotationMode=BMM_Additive
    rotationSpace=BCS_BoneSpace
    translationMode=BMM_Additive
    translationSpace=BCS_BoneSpace
    alphaBoolBlend {blendInTime=0.1, blendOutTime=0.1}
    boneToModify {boneName=C_Root_Main_Root_Jnt}
    rotation {roll=9}
    scale {x=0.95}
   }
   n17 = "Transform (Modify) Bone - Bone: C_Root_Main_Root_Jnt"(ComponentPose=@n16, bAlphaBoolEnabled=false, Translation="0,0,0", Rotation="0,0,-30", Scale="1,1,1") {
    alphaInputType=Bool
    rotationMode=BMM_Additive
    rotationSpace=BCS_BoneSpace
    translationMode=BMM_Additive
    translationSpace=BCS_BoneSpace
    alphaBoolBlend {blendInTime=0.1, blendOutTime=0.1}
    boneToModify {boneName=C_Root_Main_Root_Jnt}
    rotation {roll=-30}
   }
   n18 = "Transform (Modify) Bone - Bone: C_Root_Main_Root_Jnt"!(ComponentPose=@n17, Alpha:float=1, Translation="0,0,0", Rotation="0,0,-50", Scale="1,1,1") -> (Pose) {
    rotationMode=BMM_Additive
    rotationSpace=BCS_BoneSpace
    translationMode=BMM_Additive
    translationSpace=BCS_BoneSpace
    alphaBoolBlend {blendInTime=0.1, blendOutTime=0.1}
    alphaScaleBiasClamp {bInterpResult=true, interpSpeedDecreasing=5, interpSpeedIncreasing=5}
    boneToModify {boneName=C_Root_Main_Root_Jnt}
    rotation {roll=-50}
   }
   Output Pose(Result=Component To Local(ComponentPose=@n18))
  ]}
 EventGraph {nodes [
   Event BlueprintUpdateAnimation()
   > n1 = Sequence!(execute) -> (then_0, then_1, then_2:Exec)
   > Set isCapeSliding(isCapeSliding:bool=AND Boolean(A=Get isWearingCape() -> (isWearingCape:bool), B=Get bIsSliding())) -> (Output_Get:bool)
   Set isCapeSprinting(execute=@n1.then_1, isCapeSprinting:bool=AND Boolean(A=Get isWearingCape(), B=Get bIsTacticalSprinting())) -> (Output_Get:bool)
   n2 = Get isWearingCape()
   > Set noCapeSkydiving(noCapeSkydiving:bool=AND Boolean(A=NOT Boolean(A=@n2), B=Get bIsSkyDiving())) -> (Output_Get:bool)
   n3 = NOT Boolean(A=Get isWearingCape())
   n4 = Get bIsParachuteOpened()
   > Set noCapeParachuteOpened(noCapeParachuteOpened:double="To Float (Boolean)"(InBool=AND Boolean(A=@n3, B=@n4))) -> (Output_Get:double)
   Branch(execute=@n1.then_2, Condition=Get bIsInFrontEnd())
   n5 = GetCurveValue(self=GetOwnerFrontEndAnimInstance(), CurveName=Frontend_Idle_Bypass)
   > Set customIdleWithCape(customIdleWithCape:bool=AND Boolean(A="Equal (Float)"(A=@n5, B=1), B=Get isWearingCape())) -> (Output_Get:bool)
   Event OnCharacterCustomizationLoaded()
   n6 = CheckForBackbling()
   > Set isWearingCape(isWearingCape=OR Boolean(A=@n6.bIsBackblingACape, B=@n6.bIsBackblingLong))
  ]}
]
variables [
 customIdleWithCape {type=boolean}
 isCapeSliding {type=boolean}
 isCapeSprinting {type=boolean}
 isWearingCape {type=boolean}
 noCapeParachuteOpened {type=number}
 noCapeSkydiving {type=boolean}
]
